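; Trajectory preview actor for the birthday gift box consumable, a snap-to-grid projectile trajectory subclass that adds preview audio.
; On BeginPlay it runs parent logic, plays a preview-start cue at the instigator's location, then starts a looping sound component and attaches it to the instigator's root.
; On Destroyed it fades the loop out; overlap and tick events are disabled stubs.
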(event EventBeginPlay
  (bind _returnvalue (Game|GetInstigator))
  (bind _looping_sound (Variables|Default|GetLooping_Sound))
  (bind _rootcomponent (Variables|/Script/Engine.Actor|GetRootComponent _returnvalue))
  (|CallParentFunction)
  (Audio|PlaySoundatLocation "/Game/Sounds/Fort_Gadgets/Present/BDay/Present_BDay_PreviewStart_Cue.Present_BDay_PreviewStart_Cue" (Transformation|GetWorldLocation _rootcomponent))
  (Audio|Components|Audio|Play _looping_sound 0.0)
  (Transformation|AttachComponentToComponent _looping_sound _rootcomponent "None" "KeepRelative" "KeepRelative" "KeepRelative" true))

(event EventDestroyed
  (Audio|Components|Audio|FadeOut (Variables|Default|GetLooping_Sound) 0.5 0.0 "Linear" :!class /Script/Engine.AudioComponent))

(event Collision|EventActorBeginOverlap
  (params (OtherActor :type Actor%20Object%20Reference))
  (|CallParentFunction OtherActor) ; This node is disabled and will not be called.\nDrag off pins to build functionality.
)

(event EventTick
  (params (DeltaSeconds :type Float%20%28single-precision%29))
  (|CallParentFunction DeltaSeconds) ; This node is disabled and will not be called.\nDrag off pins to build functionality.
)

(fn ConstructionScript
  (|CallParentFunction))
