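; Projectile Blueprint for the Birthday Gift Box consumable, adding a water splash effect when it enters water.
; On the first water entry, gated by DoOnce, it queries the water body's surface info at the actor's location.
; It spawns a Niagara splash at that surface point and passes depth, body index, plane location and normal as user parameters.
; Explosion handling and construction only call the parent implementation.
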(event EventOnExploded
  (params (HitActors :type Array%20of%20Actor%20Object%20References) (HitResults :type Array%20of%20Hit%20Result%20Structures))
  (|CallParentFunction HitActors HitResults))

(event OnEnterWaterEvent(WaterInteractionComponent)
  (params (WaterBody :type Fort%20Water%20Body%20Actor%20Object%20Reference) (WaterInteractionComponent :type Fort%20Water%20Interaction%20Component%20Object%20Reference) (bIsFirstBody :type Boolean))
  (Utilities|FlowControl|DoOnce false
    (:Completed
      (bind _outwatersurfaceinfo (WaterBody|GetWaterSurfaceInfoStruct WaterBody (Transformation|GetActorLocation) true false false))
      (bind _returnvalue (Effects|Components|ParticleSystem|SpawnFXAtLocation "/Game/Athena/Items/Consumables/SuperTowerGrenade/Levels/GiftBox/BirthdayGiftBox/FX/NS_BirthdayPresent_WaterSplash.NS_BirthdayPresent_WaterSplash" (.WaterSurfaceLocation (Utilities|Struct|BreakWaterSurfaceInfo _outwatersurfaceinfo))))
      (Effects|Components|ParticleSystem|SetFloatParameter _returnvalue "User.WaterDepth" (.WaterDepth (Utilities|Struct|BreakWaterSurfaceInfo _outwatersurfaceinfo)))
      (Effects|Components|ParticleSystem|SetFloatParameter _returnvalue "User.WaterBodyIndex" (Math|Conversions|ToFloat(Integer) (.WaterBodyIdx (Utilities|Struct|BreakWaterSurfaceInfo _outwatersurfaceinfo))))
      (Effects|Components|ParticleSystem|SetVectorParameter _returnvalue "User.WaterPlaneLocation" (Utilities|Struct|BreakWaterSurfaceInfo _outwatersurfaceinfo))
      (Effects|Components|ParticleSystem|SetVectorParameter _returnvalue "User.WaterPlaneNormal" (.WaterPlaneNormal (Utilities|Struct|BreakWaterSurfaceInfo _outwatersurfaceinfo))))))

(fn ConstructionScript
  (|CallParentFunction))
