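# Unlit translucent Niagara sprite material for twinkling sparkles, named for the Sgt. Winter truck in the birthday gift box effects.
# Particle color pulses on a per-particle random sine speed, with an optional random hue shift blended in before emissive adjustments.
# A four-point star texture times particle alpha provides the transparency mask.
# A radius-adjusted distance fade scales a vertex offset toward the particle center, resizing sprites with distance.
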Material /Game/Athena/Items/Consumables/SuperTowerGrenade/Levels/GiftBox/BirthdayGiftBox/FX/M_SgtWinter_Truck_Sparkles
graph {
 properties {
  bEnableResponsiveAA=true
  bUsedWithNanite=true
  bUsedWithNiagaraSprites=true
  blendMode=BLEND_Translucent
  pixelDepthOffsetMode=PDOM_Legacy
  shadingModel=MSM_Unlit
  usedShadingModels=Unlit
 }
 nodes [
  n1 = ParticleColor()
  n2 = ParticleRandom()
  n3 = Multiply(A=@n2) {constB=0.25}
  n4 = MaterialFunctionCall(Speed=Multiply(A=@n2, B=ScalarParameter() {defaultValue=3, parameterName=TwinkleSpeed})) {materialFunction=&DebugTimeSine}
  n5 = Multiply(A=Clamp(None=Multiply(A=@n4.Result) {constB=5}) {maxDefault=5, minDefault=0.25}, B=@n1.RGB)
  MaterialFunctionCall(Emissive=LinearInterpolate(
    A=@n1.RGB
    B=MaterialFunctionCall(Hue Shift Percentage=@n3, Texture=@n5) {materialFunction=&HueShift}.Result
    Alpha=ScalarParameter() {parameterName=Use Particle Random Color}
   )) {materialFunction=&MF_FX_EmissiveAdjustments}
  n6 = ParticleRadius()
  n7 = MaterialFunctionCall(
   Position=ParticlePositionWS()
   Fade_Start_Distance=Add(A=ScalarParameter() {defaultValue=500, parameterName=FadeStartDistance}, B=@n6)
   Fade_End_Distance=Add(A=ScalarParameter() {defaultValue=200, parameterName=FadeEndDistance}, B=@n6)
  ) {materialFunction=&MF_Distance_SyncedFade_FX}
  Multiply(
   A=Subtract(A=ParticlePositionWS(), B=WorldPosition() {worldPositionShaderOffset=WPT_ExcludeAllShaderOffsets}.XYZ)
   B=OneMinus(None=LinearInterpolate(Alpha=@n7.Result) {constA=0.2})
  )
  Multiply(A=ParticleColor().A, B=TextureSample() {texture=/Game/Effects/Fort_Effects/Textures/Sparks/T_Sparkles_4Star_01_D}.G)
 ]
}
