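; Dynamic builder for the birthday gift box consumable that detects players caught inside the box once it finishes spawning.
; On the server, it estimates the box center from summed building piece locations using fixed divisors, with walls supplying XY.
; It box-overlaps pawns around that center and broadcasts an InsideBirthdayPresent custom verb for those matching the player tag container.
; The spawn-finished binding is made in the construction script and removed on EndPlay.
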(event EventEndPlay
  (params (EndPlayReason :type EEndPlayReason%20Enum :default "Destroyed"))
  (|CallParentFunction EndPlayReason)
  (EventDispatchers|UnbindEventfromBPDynamicBuilderFinishedSpawning (EventDispatchers|CreateEvent :!event CheckContainingOwner)))

(fn CheckContainingOwner
  (params (Buildings :type Array%20of%20Building%20Actor%20Object%20References))
  (Networking|SwitchHasAuthority
    (:Authority
      (for _array_element Buildings
        (LocalVariables|SetlocBuilding (Transformation|GetActorLocation _array_element))
        (switch Utilities|FlowControl|Switch|SwitchonEFortBuildingType (Variables|/Script/FortniteGame.BuildingActor|GetBuildingType _array_element)
          (:Wall
            (LocalVariables|SetcenterXY (+ (LocalVariables|GetcenterXY) (Math|Vector|MakeVector (.x (LocalVariables|GetlocBuilding)) (.y (LocalVariables|GetlocBuilding))))))
          (:Floor
            _)
          (:Roof
            _))
        (LocalVariables|SetcenterZ (+ (LocalVariables|GetcenterZ) (Math|Vector|MakeVector 0.0 0.0 (.z (LocalVariables|GetlocBuilding))))))
      (bind _output_get (LocalVariables|SetcenterOfMass (+ (/ (LocalVariables|GetcenterXY) 4.0) (/ (LocalVariables|GetcenterZ) 2.0))))
      (bind _outactors (Collision|BoxOverlapActors _output_get "250.000000,250.000000,200.000000" (Utilities|Array|MakeArray "ObjectTypeQuery3") "/Script/FortniteGame.FortPawn" :!class /Script/Engine.KismetSystemLibrary))
      (for _array_element_2 _outactors
        (bind _fortpawn (Class|FortQuestBlueprintLibrary|GetFortPawnFromObject _array_element_2)
          (:Valid
            (if (GameplayTags|HasAnyTags (GameplayTags|GetOwnedGameplayTags _fortpawn :!class /Script/GameplayTags.BlueprintGameplayTagLibrary) (Variables|Gettc_players))
              (CustomVerb|BroadcastCustomVerbMessage (CustomVerb|MakeCustomVerbMessage _array_element_2 "(TagName=\"CustomVerb.InsideBirthdayPresent\")")))))))))

(fn ConstructionScript
  (|CallParentFunction)
  (EventDispatchers|BindEventtoBPDynamicBuilderFinishedSpawning (EventDispatchers|CreateEvent :!event CheckContainingOwner)))
